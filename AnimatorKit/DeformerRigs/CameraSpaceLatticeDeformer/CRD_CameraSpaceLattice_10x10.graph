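# Control Rig deformer that warps meshes through a 10x10 control lattice on a camera's near plane.
# Construction spawns the CameraLattice root, its channels and the point grid from reference FOV and aspect.
# Forward solve rescales CameraSpace from CameraHandle lens channels, draws frustum and lattice, and feeds points and Weight to DG_CameraSpaceLattice.
# Optional soft selection spreads single-point drags with smoothstep falloff, using stored previous transforms, and can auto-key affected points.
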ControlRigBlueprint /AnimatorKit/DeformerRigs/CameraSpaceLatticeDeformer/CRD_CameraSpaceLattice_10x10
graphs [
 Concat_1_ContainedGraph {nodes [
   n1 = RigVMFunctionEntryNode()
   Concat = /Script/RigVM.RigVMFunction_NameConcat(A=@n1.A, B=@n1.B) {is=struct RigVMUnitNode, template="Concat::Execute(in A,in B,out Result)"}
   RigVMFunctionReturnNode(Result=&RigVMFunction_NameConcat(A=@Concat.Result, B=@n1.C) {template="Concat::Execute(in A,in B,out Result)"}.Result)
  ]}
 Multiply_6_1_ContainedGraph {nodes [
   n1 = RigVMFunctionEntryNode()
   RigVMFunctionReturnNode(Result=&RigVMFunction_MathDoubleMul(B=@n1.C, A=&RigVMFunction_MathDoubleMul(A=@n1.A, B=@n1.B).Result).Result)
  ]}
 Multiply_6_1_ContainedGraph {nodes [
   n1 = RigVMFunctionEntryNode()
   RigVMFunctionReturnNode(Result=&RigVMFunction_MathDoubleMul(B=@n1.C, A=&RigVMFunction_MathDoubleMul(A=@n1.A, B=@n1.B).Result).Result)
  ]}
 New Function_ContainedGraph {nodes [
   n1 = RigVMFunctionEntryNode()
   > RigVMFunctionReturnNode(
    Index2d=(X=0,Y=0,Z=0)
    Index2d.Y=/Script/RigVM.RigVMFunction_MathIntDiv(A=@n1.Index, B=@n1.NumControls.X) {is=struct RigVMUnitNode, template="Divide::Execute(in A,in B,out Result)"}.Result
    Index2d.X=/Script/RigVM.RigVMFunction_MathIntMod(A=@n1.Index, B=@n1.NumControls.X) {is=struct RigVMUnitNode, template="Modulo::Execute(in A,in B,out Result)"}.Result
   )
  ]}
 New Function_ContainedGraph {nodes [
   n1 = RigVMFunctionEntryNode()
   > RigVMFunctionReturnNode(Index=/Script/RigVM.RigVMFunction_MathIntAdd(
     A=@n1.Index2d.X
     B=/Script/RigVM.RigVMFunction_MathIntMul(A=@n1.Index2d.Y, B=@n1.NumPoints.X) {is=struct RigVMUnitNode, template="Multiply::Execute(in A,in B,out Result)"}.Result
    ) {is=struct RigVMUnitNode, template="Add::Execute(in A,in B,out Result)"}.Result)
  ]}
 New Function_ContainedGraph {nodes [
   n1 = RigVMFunctionEntryNode()
   Power = /Script/RigVM.RigVMFunction_MathDoublePow(B=2, A=@n1.A) {is=struct RigVMUnitNode, template="Power::Execute(in A,in B,out Result)"}
   n2 = &RigVMFunction_MathDoublePow(B=3, A=@n1.A) {template="Power::Execute(in A,in B,out Result)"}
   > RigVMFunctionReturnNode(Result=&RigVMFunction_MathDoubleSub(
     A=&RigVMFunction_MathDoubleMul(A=3, B=@Power.Result).Result
     B=&RigVMFunction_MathDoubleMul(A=2, B=@n2.Result).Result
    ).Result)
  ]}
 RigVMFunctionLibrary {nodes [
   FromIndex2d = RigVMCollapseNode(Index2d=(X=0,Y=0,Z=0), NumPoints=(X=0,Y=0,Z=0))
   SmoothStep = RigVMCollapseNode(A=1)
   ToIndex2d = RigVMCollapseNode(Index=0, NumControls=(X=0,Y=0,Z=0))
  ]}
 RigVMModel {nodes [
   n1 = &RigUnit_PrepareForExecution()
   n2 = &RigVMFunction_MathDoubleDiv(B=2, A=RigVMVariableNode() {variable=ReferenceFieldOfView}.Value)
   n3 = &RigVMFunction_MathDoubleMul(
    A=/Script/RigVM.RigVMFunction_MathDoubleTan(Value=/Script/RigVM.RigVMFunction_MathDoubleRad(Value=@n2.Result) {is=struct RigVMUnitNode, template="Radians::Execute(in Value,out Result)"}.Result) {is=struct RigVMUnitNode, template="Tan::Execute(in Value,out Result)"}.Result
    B=RigVMVariableNode() {variable=ZNear}.Value
   )
   RigVMVariableNode(Value=@n3.Result, ExecuteContext=@n1.ExecutePin) {variable=ReferenceHalfWidth}
   n4 = RigVMVariableNode() {variable=ReferenceAspectRatio}
   > RigVMVariableNode(Value=&RigVMFunction_MathDoubleMul(A=&RigVMFunction_MathDoubleDiv(B=@n4.Value.X, A=@n4.Value.Y).Result, B=@n3.Result).Result) {variable=ReferenceHalfHeight}
   > n5 = /Script/ControlRig.RigUnit_HierarchyAddControlTransform(
    InitialValue=(Rotation=(X=0,Y=0,Z=0,W=1),Translation=(X=0,Y=0,Z=0),Scale3D=(X=1,Y=1,Z=1))
    OffsetSpace=LocalSpace
    OffsetTransform=(Rotation=(X=0,Y=0,Z=0,W=1),Translation=(X=0,Y=0,Z=0),Scale3D=(X=1,Y=1,Z=1))
    Parent=(Type=Null,Name="CameraSpace")
    Settings="(InitialSpace=LocalSpace,bUsePreferredRotationOrder=False,PreferredRotationOrder=YZX,Limits=(LimitTranslationX=(bMinimum=False,bMaximum=False),LimitTranslationY=(bMinimum=False,bMaximum=False),LimitTranslationZ=(bMinimum=true,bMaximum=true),LimitPitch=(bMi ...(truncated)"
    Settings.Shape.Transform.Translation.X=&RigVMFunction_MathDoubleMul(B=-0.88, A=RigVMVariableNode() {variable=ReferenceHalfWidth}.Value).Result
    OffsetTransform.Translation.Z=RigVMVariableNode() {variable=ZNear}.Value
    Name=RigVMVariableNode() {variable=LatticeRootControlName}.Value
   ) {
    is=struct RigVMUnitNode
    template="SpawnControl::Execute(in InitialValue,in Settings,in OffsetTransform,in OffsetSpace,in Parent,in Name,out Item)"
   }
   > SpawnAnimationChannel = /Script/ControlRig.RigUnit_HierarchyAddAnimationChannelFloat(InitialValue=1, LimitsEnabled=(Enabled=(bMinimum=True,bMaximum=True)), MaximumValue=1, MinimumValue=0, Name=Weight, Parent=@n5.Item) {
    is=struct RigVMUnitNode
    template="SpawnAnimationChannel::Execute(in InitialValue,in MinimumValue,in MaximumValue,in LimitsEnabled,in Parent,in Name,out Item)"
   }
   > /Script/ControlRig.RigUnit_HierarchyAddAnimationChannelBool(InitialValue=false, LimitsEnabled=(), MaximumValue=True, MinimumValue=False, Name=EnableSoftSelection, Parent=@n5.Item) {
    is=struct RigVMUnitNode
    template="SpawnAnimationChannel::Execute(in InitialValue,in MinimumValue,in MaximumValue,in LimitsEnabled,in Parent,in Name,out Item)"
   }
   > &RigUnit_HierarchyAddAnimationChannelFloat(
    InitialValue=0.2
    LimitsEnabled=(Enabled=(bMinimum=True,bMaximum=true))
    MaximumValue=1
    MinimumValue=0
    Name=SoftSelectionRadius
    Parent=@n5.Item
   ) {template="SpawnAnimationChannel::Execute(in InitialValue,in MinimumValue,in MaximumValue,in LimitsEnabled,in Parent,in Name,out Item)"}
   > &RigUnit_HierarchyAddAnimationChannelBool(InitialValue=true, LimitsEnabled=(), MaximumValue=True, MinimumValue=False, Name=EnableAutoKeySoftSelection, Parent=@n5.Item) {template="SpawnAnimationChannel::Execute(in InitialValue,in MinimumValue,in MaximumValue,in LimitsEnabled,in Parent,in Name,out Item)"}
   > Set_Metadata_2_1 = RigVMDispatchNode(
    Item=(Type=Control,Name="CameraLattice")
    Name=PreviousSoftSelectionRadius
    NameSpace=Self
    Value=/Script/ControlRig.RigUnit_GetFloatAnimationChannel(Channel=SoftSelectionRadius, Control=CameraLattice, bInitial=False) {is=struct RigVMUnitNode, template="GetAnimationChannel::Execute(out Value,in Control,in Channel,in bInitial)"}.Value
   ) {template="DISPATCH_RigDispatch_SetMetadata(in Item,in Name,in NameSpace,in Value,out Success)"}
   > &RigUnit_HierarchyAddControlTransform(
    InitialValue=(Rotation=(X=0,Y=0,Z=0,W=1),Translation=(X=0,Y=0,Z=0),Scale3D=(X=1,Y=1,Z=1))
    Name=Points
    OffsetSpace=LocalSpace
    OffsetTransform=(Rotation=(X=0,Y=0,Z=0,W=1),Translation=(X=0,Y=0,Z=0),Scale3D=(X=1,Y=1,Z=1))
    Settings="(InitialSpace=LocalSpace,bUsePreferredRotationOrder=False,PreferredRotationOrder=YZX,Limits=(LimitTranslationX=(bMinimum=False,bMaximum=False),LimitTranslationY=(bMinimum=False,bMaximum=False),LimitTranslationZ=(bMinimum=False,bMaximum=False),LimitPitch=(b ...(truncated)"
    Parent=@n5.Item
   ) {template="SpawnControl::Execute(in InitialValue,in Settings,in OffsetTransform,in OffsetSpace,in Parent,in Name,out Item)"}
   n6 = RigVMVariableNode() {variable=NumPoints}
   > RigVMFunction_ForLoopCount = /Script/RigVM.RigVMFunction_ForLoopCount(Count=@n6.Value.Y) {is=struct RigVMUnitNode}
   > n7 = &RigVMFunction_ForLoopCount(Count=@n6.Value.X)
   Int to Name = /Script/RigVM.RigVMFunction_MathIntToName(PaddedSize=2, Number=@n7.Index) {is=struct RigVMUnitNode, template="Int to Name::Execute(in Number,in PaddedSize,out Result)"}
   Int to Name_1 = &RigVMFunction_MathIntToName(PaddedSize=2, Number=@RigVMFunction_ForLoopCount.Index) {template="Int to Name::Execute(in Number,in PaddedSize,out Result)"}
   Concat = RigVMAggregateNode(A=Point, B=@"Int to Name".Result, C=@"Int to Name_1".Result)
   n8 = RigVMVariableNode() {variable=ReferenceHalfWidth}
   Multiply = &RigVMFunction_MathDoubleMul(B=&RigVMFunction_MathDoubleMul(B=2, A=@n8.Value).Result, A=@n7.Ratio)
   Subtract = &RigVMFunction_MathDoubleSub(A=@Multiply.Result, B=@n8.Value)
   n9 = RigVMVariableNode() {variable=ReferenceHalfHeight}
   n10 = &RigVMFunction_MathDoubleMul(B=2, A=@n9.Value)
   > SpawnControl = &RigUnit_HierarchyAddControlTransform(
    InitialValue=(Rotation=(X=0,Y=0,Z=0,W=1),Translation=(X=0,Y=0,Z=0),Scale3D=(X=1,Y=1,Z=1))
    OffsetSpace=LocalSpace
    OffsetTransform=(Rotation=(X=0,Y=0,Z=0,W=1),Translation=(X=0,Y=0,Z=0),Scale3D=(X=1,Y=1,Z=1))
    Parent=(Type=Control,Name="Points")
    Settings="(InitialSpace=LocalSpace,bUsePreferredRotationOrder=False,PreferredRotationOrder=YZX,Limits=(LimitTranslationX=(bMinimum=False,bMaximum=False),LimitTranslationY=(bMinimum=False,bMaximum=False),LimitTranslationZ=(bMinimum=true,bMaximum=true),LimitPitch=(bMi ...(truncated)"
    Name=@Concat.Result
    OffsetTransform.Translation.X=@Subtract.Result
    OffsetTransform.Translation.Y=&RigVMFunction_MathDoubleSub(A=&RigVMFunction_MathDoubleMul(B=@n10.Result, A=@RigVMFunction_ForLoopCount.Ratio).Result, B=@n9.Value).Result
    Settings.Shape.Color=RigVMVariableNode() {variable=LatticePointColor}.Value
   ) {template="SpawnControl::Execute(in InitialValue,in Settings,in OffsetTransform,in OffsetSpace,in Parent,in Name,out Item)"}
   > Set_Metadata_2_2 = RigVMDispatchNode(Name=WasInteracted, NameSpace=Self, Value=False, Item=@SpawnControl.Item) {template="DISPATCH_RigDispatch_SetMetadata(in Item,in Name,in NameSpace,in Value,out Success)"}
   RigUnit_BeginExecution = &RigUnit_BeginExecution()
   n11 = &RigVMFunction_Sequence(ExecuteContext=@RigUnit_BeginExecution.ExecutePin)
   n12 = &RigVMFunction_Sequence(ExecuteContext=@n11.A)
   n13 = /Script/ControlRig.RigUnit_IsInteracting() {is=struct RigVMUnitNode}
   n14 = &RigVMFunction_ControlFlowBranch(Condition=@n13.bIsInteracting, ExecuteContext=@n12.A)
   For_Each_3 = RigVMDispatchNode(
    Array=/Script/ControlRig.RigUnit_CollectionChildrenArray(Parent=(Type=Control,Name="Points"), TypeToSearch=All, bDefaultChildren=True, bIncludeParent=False, bRecursive=False) {is=struct RigVMUnitNode}.Items
    ExecuteContext=@n14.False
   ) {template="DISPATCH_RigVMDispatch_ArrayIterator(in Array,out Element,out Index,out Count,out Ratio)"}
   > &RigUnit_SetTransform(
    Space=LocalSpace
    Value=(Rotation=(X=0,Y=0,Z=0,W=1),Translation=(X=0,Y=0,Z=0),Scale3D=(X=1,Y=1,Z=1))
    Weight=1
    bInitial=False
    Item=@For_Each_3.Element
    Value.Translation=&RigUnit_GetTransform(Space=LocalSpace, bInitial=False, Item=@For_Each_3.Element).Transform.Translation
   )
   > Set_Metadata = RigVMDispatchNode(
    Name=PreviousTransform
    NameSpace=Self
    Value=&RigUnit_GetTransform(bInitial=False, Item=@For_Each_3.Element).Transform
    Item=@For_Each_3.Element
   ) {template="DISPATCH_RigDispatch_SetMetadata(in Item,in Name,in NameSpace,in Value,out Success)"}
   Get_Metadata_4 = RigVMDispatchNode(Default=False, Name=WasInteracted, NameSpace=Self, Item=@For_Each_3.Element) {template="DISPATCH_RigDispatch_GetMetadata(in Item,in Name,in NameSpace,in Default,out Value,out Found)"}
   n15 = &RigVMFunction_ControlFlowBranch(Condition=@Get_Metadata_4.Value, ExecuteContext=@Set_Metadata.ExecuteContext)
   SendEvent = /Script/ControlRig.RigUnit_SendEvent(
    Event=RequestAutoKey
    OffsetInSeconds=0
    bOnlyDuringInteraction=false
    bEnable=/Script/ControlRig.RigUnit_GetBoolAnimationChannel(Channel=EnableAutoKeySoftSelection, Control=CameraLattice, bInitial=False) {is=struct RigVMUnitNode, template="GetAnimationChannel::Execute(out Value,in Control,in Channel,in bInitial)"}.Value
    ExecutePin=@n15.True
    Item=@For_Each_3.Element
   ) {is=struct RigVMUnitNode, template="SendEvent::Execute(in Event,in Item,in OffsetInSeconds,in bEnable,in bOnlyDuringInteraction)"}
   > Set_Metadata_3 = RigVMDispatchNode(Name=WasInteracted, NameSpace=Self, Value=False, Item=@For_Each_3.Element) {template="DISPATCH_RigDispatch_SetMetadata(in Item,in Name,in NameSpace,in Value,out Success)"}
   RigUnit_GetBoolAnimationChannel = &RigUnit_GetBoolAnimationChannel(Channel=EnableSoftSelection, Control=CameraLattice, bInitial=False) {template="GetAnimationChannel::Execute(out Value,in Control,in Channel,in bInitial)"}
   n16 = &RigVMFunction_ControlFlowBranch(Condition=@RigUnit_GetBoolAnimationChannel.Value, ExecuteContext=@n14.True)
   n17 = &RigVMFunction_ControlFlowBranch(
    Condition=RigVMDispatchNode(B=1, A=RigVMDispatchNode(Array=@n13.Items) {template="DISPATCH_RigVMDispatch_ArrayGetNum(in Array,out Num)"}.Num) {template="DISPATCH_RigVMDispatch_CoreEquals(in A,in B,out Result)"}.Result
    ExecuteContext=@n16.True
   )
   n18 = &RigVMFunction_ControlFlowBranch(Condition=false, ExecuteContext=@n17.False)
   For_Each_4 = RigVMDispatchNode(
    Array=&RigUnit_CollectionChildrenArray(Parent=(Type=Control,Name="Points"), TypeToSearch=All, bDefaultChildren=True, bIncludeParent=False, bRecursive=False).Items
    ExecuteContext=@n18.True
   ) {template="DISPATCH_RigVMDispatch_ArrayIterator(in Array,out Element,out Index,out Count,out Ratio)"}
   n19 = &RigVMFunction_ControlFlowBranch(
    Condition=RigVMDispatchNode(Element=@For_Each_4.Element, Array=@n13.Items) {template="DISPATCH_RigVMDispatch_ArrayFind(in Array,in Element,out Index,out Success)"}.Success
    ExecuteContext=@For_Each_4.ExecuteContext
   )
   Get_Metadata_2 = RigVMDispatchNode(
    Default=(Rotation=(X=0,Y=0,Z=0,W=1),Translation=(X=0,Y=0,Z=0),Scale3D=(X=1,Y=1,Z=1))
    Name=PreviousTransform
    NameSpace=Self
    Item=@For_Each_4.Element
   ) {template="DISPATCH_RigDispatch_GetMetadata(in Item,in Name,in NameSpace,in Default,out Value,out Found)"}
   &RigUnit_SetTransform(
    Value=(Rotation=(X=0,Y=0,Z=0,W=1),Translation=(X=0,Y=0,Z=0),Scale3D=(X=1,Y=1,Z=1))
    Weight=1
    bInitial=False
    Item=@For_Each_4.Element
    Value.Translation=@Get_Metadata_2.Value.Translation
    ExecutePin=@n19.False
   )
   n20 = RigVMDispatchNode(Index=0, Array=@n13.Items) {template="DISPATCH_RigVMDispatch_ArrayGetAtIndex(in Array,in Index,out Element)"}
   n21 = &RigVMFunction_ControlFlowBranch(
    Condition=RigVMDispatchNode(
     B=(Type=Control,Name="Points")
     A=/Script/ControlRig.RigUnit_HierarchyGetParent(bDefaultParent=True, Child=@n20.Element) {is=struct RigVMUnitNode}.Parent
    ) {template="DISPATCH_RigVMDispatch_CoreEquals(in A,in B,out Result)"}.Result
    ExecuteContext=@n17.True
   )
   For_Each_2 = RigVMDispatchNode(
    Array=&RigUnit_CollectionChildrenArray(Parent=(Type=Control,Name="Points"), TypeToSearch=All, bDefaultChildren=True, bIncludeParent=False, bRecursive=False).Items
    ExecuteContext=@n21.True
   ) {template="DISPATCH_RigVMDispatch_ArrayIterator(in Array,out Element,out Index,out Count,out Ratio)"}
   n22 = RigVMDispatchNode(Index=0, Array=@n13.Items) {template="DISPATCH_RigVMDispatch_ArrayGetAtIndex(in Array,in Index,out Element)"}
   Not_Equals = RigVMDispatchNode(B=@n22.Element, A=@For_Each_2.Element) {template="DISPATCH_RigVMDispatch_CoreNotEquals(in A,in B,out Result)"}
   n23 = &RigVMFunction_ControlFlowBranch(ExecuteContext=@For_Each_2.ExecuteContext, Condition=@Not_Equals.Result)
   Set_Metadata_2_3 = RigVMDispatchNode(Name=WasInteracted, NameSpace=Self, Value=true, ExecuteContext=@n23.False, Item=@For_Each_2.Element) {template="DISPATCH_RigDispatch_SetMetadata(in Item,in Name,in NameSpace,in Value,out Success)"}
   Get_Metadata_1_1 = RigVMDispatchNode(
    Default=(Rotation=(X=0,Y=0,Z=0,W=1),Translation=(X=0,Y=0,Z=0),Scale3D=(X=1,Y=1,Z=1))
    Name=PreviousTransform
    NameSpace=Self
    Item=@n22.Element
   ) {template="DISPATCH_RigDispatch_GetMetadata(in Item,in Name,in NameSpace,in Default,out Value,out Found)"}
   n24 = &RigUnit_GetTransform(Item=(Type=Null,Name="CameraSpace"), bInitial=False)
   n25 = &RigVMFunction_MathTransformMakeRelative(Global=@Get_Metadata_1_1.Value, Parent=@n24.Transform)
   n26 = &RigUnit_GetTransform(bInitial=False, Item=@n22.Element)
   n27 = &RigUnit_GetTransform(Item=(Type=Null,Name="CameraSpace"), bInitial=False)
   Get_Metadata = RigVMDispatchNode(
    Default=(Rotation=(X=0,Y=0,Z=0,W=1),Translation=(X=0,Y=0,Z=0),Scale3D=(X=1,Y=1,Z=1))
    Name=PreviousTransform
    NameSpace=Self
    Item=@n22.Element
   ) {template="DISPATCH_RigDispatch_GetMetadata(in Item,in Name,in NameSpace,in Default,out Value,out Found)"}
   Multiply_6 = RigVMAggregateNode(
    C=2
    A=&RigUnit_GetFloatAnimationChannel(Channel=SoftSelectionRadius, Control=CameraLattice, bInitial=False) {template="GetAnimationChannel::Execute(out Value,in Control,in Channel,in bInitial)"}.Value
    B=RigVMVariableNode() {variable=ReferenceHalfWidth}.Value
   )
   Get_Metadata_1 = RigVMDispatchNode(
    Default=(Rotation=(X=0,Y=0,Z=0,W=1),Translation=(X=0,Y=0,Z=0),Scale3D=(X=1,Y=1,Z=1))
    Name=PreviousTransform
    NameSpace=Self
    Item=@For_Each_2.Element
   ) {template="DISPATCH_RigDispatch_GetMetadata(in Item,in Name,in NameSpace,in Default,out Value,out Found)"}
   n28 = &RigVMFunction_MathTransformMakeRelative(Global=@Get_Metadata_1.Value, Parent=@n24.Transform)
   n29 = &RigVMFunction_MathTransformMakeRelative(
    Global=(Rotation=(X=0,Y=0,Z=0,W=1),Translation=(X=0,Y=0,Z=0),Scale3D=(X=1,Y=1,Z=1))
    Parent=(Rotation=(X=0,Y=0,Z=0,W=1),Translation=(X=0,Y=0,Z=0),Scale3D=(X=1,Y=1,Z=1))
    Global.Translation=@n28.Local.Translation
    Parent.Translation=@n25.Local.Translation
   )
   RigVMFunction_MathVectorLength = /Script/RigVM.RigVMFunction_MathVectorLength(Value=@n29.Local.Translation) {is=struct RigVMUnitNode, template="Length::Execute(in Value,out Result)"}
   Clamp = &RigVMFunction_MathDoubleClamp(Maximum=@Multiply_6.Result, Value=@RigVMFunction_MathVectorLength.Result)
   SmoothStep = RigVMFunctionReferenceNode(A=&RigVMFunction_MathDoubleSub(A=1, B=&RigVMFunction_MathDoubleDiv(A=@Clamp.Result, B=@Multiply_6.Result).Result).Result)
   Interpolate = /Script/RigVM.RigVMFunction_MathTransformLerp(
    A=(Rotation=(X=0,Y=0,Z=0,W=1),Translation=(X=0,Y=0,Z=0),Scale3D=(X=1,Y=1,Z=1))
    B=&RigVMFunction_MathTransformMakeRelative(
     Global=&RigVMFunction_MathTransformMakeRelative(Global=@n26.Transform, Parent=@n27.Transform).Local
     Parent=&RigVMFunction_MathTransformMakeRelative(Global=@Get_Metadata.Value, Parent=@n27.Transform).Local
    ).Local
    T=@SmoothStep.Result
   ) {is=struct RigVMUnitNode, template="Interpolate::Execute(in A,in B,in T,out Result)"}
   n30 = &RigVMFunction_MathTransformMul(B=@Interpolate.Result, A=&RigVMFunction_MathTransformMakeRelative(Global=@n28.Local, Parent=@n25.Local).Local) {template="Multiply::Execute(in A,in B,out Result)"}
   &RigUnit_SetTransform(
    Weight=1
    bInitial=False
    Value=/Script/RigVM.RigVMFunction_MathTransformMakeAbsolute(
     Local=&RigVMFunction_MathTransformMakeAbsolute(Parent=@n25.Local, Local=@n30.Result) {template="Make Absolute::Execute(in Local,in Parent,out Global)"}.Global
     Parent=&RigUnit_GetTransform(Item=(Type=Null,Name="CameraSpace"), bInitial=False).Transform
    ) {is=struct RigVMUnitNode, template="Make Absolute::Execute(in Local,in Parent,out Global)"}.Global
    ExecutePin=@n23.True
    Item=@For_Each_2.Element
   )
   > Set_Metadata_2 = RigVMDispatchNode(
    Name=WasInteracted
    NameSpace=Self
    Value=/Script/RigVM.RigVMFunction_MathDoubleGreater(B=0, A=@SmoothStep.Result) {is=struct RigVMUnitNode, template="Greater::Execute(in A,in B,out Result)"}.Result
    Item=@For_Each_2.Element
   ) {template="DISPATCH_RigDispatch_SetMetadata(in Item,in Name,in NameSpace,in Value,out Success)"}
   n31 = &RigUnit_GetTransform(Item=(Type=Null,Name="CameraSpace"), bInitial=False)
   RigVMFunction_DebugArcNoSpace = /Script/RigVM.RigVMFunction_DebugArcNoSpace(
    Color=(R=1,G=0,B=0,A=1)
    DebugDrawSettings=(DepthPriority=SDPG_Foreground,Lifetime=-1)
    Detail=32
    MaximumDegrees=360
    MinimumDegrees=0
    Thickness=0
    Transform=(Rotation=(X=0,Y=0,Z=0,W=1),Translation=(X=0,Y=0,Z=0),Scale3D=(X=1,Y=1,Z=1))
    bEnabled=True
    ExecutePin=@For_Each_2.Completed
    Transform.Translation=&RigVMFunction_MathTransformMakeRelative(Global=&RigUnit_GetTransform(bInitial=False, Item=@n22.Element).Transform, Parent=@n31.Transform).Local.Translation
    Radius=@Multiply_6.Result
    WorldOffset=@n31.Transform
   ) {is=struct RigVMUnitNode}
   n32 = &RigVMFunction_Sequence(ExecuteContext=@n12.B)
   n33 = &RigVMFunction_ControlFlowBranch(
    Condition=&RigUnit_GetBoolAnimationChannel(Channel=EnableSoftSelection, Control=CameraLattice, bInitial=False) {template="GetAnimationChannel::Execute(out Value,in Control,in Channel,in bInitial)"}.Value
    ExecuteContext=@n32.A
   )
   Greater = &RigVMFunction_MathDoubleGreater(B=0, A=RigVMVariableNode() {variable=RadiusIndicatorTimeLeft}.Value) {template="Greater::Execute(in A,in B,out Result)"}
   n34 = &RigVMFunction_ControlFlowBranch(Condition=@Greater.Result, ExecuteContext=@n33.Completed)
   RigVMFunction_GetDeltaTime = /Script/RigVM.RigVMFunction_GetDeltaTime() {is=struct RigVMUnitNode}
   RigVMVariableNode(
    ExecuteContext=@n34.True
    Value=&RigVMFunction_MathDoubleSub(B=@RigVMFunction_GetDeltaTime.Result, A=RigVMVariableNode() {variable=RadiusIndicatorTimeLeft}.Value).Result
   ) {variable=RadiusIndicatorTimeLeft}
   Multiply_6_1 = RigVMAggregateNode(
    C=2
    A=&RigUnit_GetFloatAnimationChannel(Channel=SoftSelectionRadius, Control=CameraLattice, bInitial=False) {template="GetAnimationChannel::Execute(out Value,in Control,in Channel,in bInitial)"}.Value
    B=RigVMVariableNode() {variable=ReferenceHalfWidth}.Value
   )
   > &RigVMFunction_DebugArcNoSpace(
    Color=(R=1,G=0,B=0,A=1)
    DebugDrawSettings=(DepthPriority=SDPG_Foreground,Lifetime=-1)
    Detail=32
    MaximumDegrees=360
    MinimumDegrees=0
    Thickness=0
    Transform=(Rotation=(X=0,Y=0,Z=0,W=1),Translation=(X=0,Y=0,Z=0),Scale3D=(X=1,Y=1,Z=1))
    bEnabled=True
    Radius=@Multiply_6_1.Result
    WorldOffset=&RigUnit_GetTransform(Item=(Type=Null,Name="CameraSpace"), bInitial=False).Transform
    Transform.Translation.Z=RigVMVariableNode() {variable=ZNear}.Value
   )
   Get_Metadata_3 = RigVMDispatchNode(Default=0, Item=(Type=Control,Name="CameraLattice"), Name=PreviousSoftSelectionRadius, NameSpace=Self) {template="DISPATCH_RigDispatch_GetMetadata(in Item,in Name,in NameSpace,in Default,out Value,out Found)"}
   n35 = &RigUnit_GetFloatAnimationChannel(Channel=SoftSelectionRadius, Control=CameraLattice, bInitial=False) {template="GetAnimationChannel::Execute(out Value,in Control,in Channel,in bInitial)"}
   Not_Equals_1 = RigVMDispatchNode(A=@Get_Metadata_3.Value, B=@n35.Value) {template="DISPATCH_RigVMDispatch_CoreNotEquals(in A,in B,out Result)"}
   n36 = &RigVMFunction_ControlFlowBranch(Condition=@Not_Equals_1.Result, ExecuteContext=@n33.True)
   Set_Metadata_1 = RigVMDispatchNode(Item=(Type=Control,Name="CameraLattice"), Name=PreviousSoftSelectionRadius, NameSpace=Self, Value=@n35.Value, ExecuteContext=@n36.True) {template="DISPATCH_RigDispatch_SetMetadata(in Item,in Name,in NameSpace,in Value,out Success)"}
   > RigVMVariableNode(Value=2) {variable=RadiusIndicatorTimeLeft}
   n37 = &RigVMFunction_Sequence(ExecuteContext=@n11.B)
   n38 = &RigVMFunction_MathDoubleDiv(
    B=2
    A=&RigUnit_GetFloatAnimationChannel(Channel=FieldOfView, Control=CameraHandle, bInitial=False) {template="GetAnimationChannel::Execute(out Value,in Control,in Channel,in bInitial)"}.Value
   )
   n39 = &RigVMFunction_MathDoubleMul(
    A=&RigVMFunction_MathDoubleTan(Value=&RigVMFunction_MathDoubleRad(Value=@n38.Result) {template="Radians::Execute(in Value,out Result)"}.Result) {template="Tan::Execute(in Value,out Result)"}.Result
    B=RigVMVariableNode() {variable=ZNear}.Value
   )
   Divide = &RigVMFunction_MathDoubleDiv(A=@n39.Result, B=RigVMVariableNode() {variable=ReferenceHalfWidth}.Value)
   n40 = &RigVMFunction_MathDoubleDiv(
    B=&RigUnit_GetFloatAnimationChannel(Channel=AspectRatioX, Control=CameraHandle, bInitial=False) {template="GetAnimationChannel::Execute(out Value,in Control,in Channel,in bInitial)"}.Value
    A=&RigUnit_GetFloatAnimationChannel(Channel=AspectRatioY, Control=CameraHandle, bInitial=False) {template="GetAnimationChannel::Execute(out Value,in Control,in Channel,in bInitial)"}.Value
   )
   RigUnit_SetTransform = &RigUnit_SetTransform(
    Item=(Type=Null,Name="CameraSpace")
    Space=LocalSpace
    Value=(Rotation=(X=0.5,Y=0.5,Z=0.5,W=0.5),Translation=(X=0,Y=0,Z=0),Scale3D=(X=1,Y=1,Z=1))
    Weight=1
    bInitial=False
    Value.Scale3D.X=@Divide.Result
    Value.Scale3D.Y=&RigVMFunction_MathDoubleDiv(A=&RigVMFunction_MathDoubleMul(A=@n40.Result, B=@n39.Result).Result, B=RigVMVariableNode() {variable=ReferenceHalfHeight}.Value).Result
    ExecutePin=@n37.A
   )
   RigVMFunction_Sequence = &RigVMFunction_Sequence(ExecuteContext=@n37.B)
   n41 = &RigVMFunction_MathDoubleAdd(B=-0.1, A=RigVMVariableNode() {variable=ZNear}.Value)
   n42 = RigVMVariableNode() {variable=ReferenceHalfHeight}
   Negate = /Script/RigVM.RigVMFunction_MathDoubleNegate(Value=@n42.Value) {is=struct RigVMUnitNode, template="Negate::Execute(in Value,out Result)"}
   n43 = RigVMVariableNode() {variable=ReferenceHalfWidth}
   n44 = &RigVMFunction_MathDoubleNegate(Value=@n43.Value) {template="Negate::Execute(in Value,out Result)"}
   n45 = RigVMDispatchNode(Value=(X=0,Y=0,Z=0), Value.Z=@n41.Result, Value.Y=@Negate.Result, Value.X=@n44.Result) {template="DISPATCH_RigVMDispatch_Constant(io Value)"}
   n46 = RigVMDispatchNode(Value=(X=0,Y=0,Z=0), Value.Z=@n41.Result, Value.X=@n44.Result, Value.Y=@n42.Value) {template="DISPATCH_RigVMDispatch_Constant(io Value)"}
   n47 = RigVMDispatchNode(Value=(X=0,Y=0,Z=0), Value.Z=@n41.Result, Value.X=@n43.Value, Value.Y=@n42.Value) {template="DISPATCH_RigVMDispatch_Constant(io Value)"}
   n48 = RigVMDispatchNode(Value=(X=0,Y=0,Z=0), Value.Z=@n41.Result, Value.Y=@Negate.Result, Value.X=@n43.Value) {template="DISPATCH_RigVMDispatch_Constant(io Value)"}
   n49 = &RigUnit_GetTransform(Item=(Type=Null,Name="CameraSpace"), bInitial=False)
   RigVMFunction_DebugLineStripNoSpace = /Script/RigVM.RigVMFunction_DebugLineStripNoSpace(
    Color=(R=1,G=0,B=0,A=1)
    DebugDrawSettings=(DepthPriority=SDPG_Foreground,Lifetime=-1)
    Points=((X=0,Y=0,Z=0),(X=0,Y=0,Z=0),(X=0,Y=0,Z=0),(X=0,Y=0,Z=0),(X=0,Y=0,Z=0))
    Thickness=0
    bEnabled=true
    Points.0=@n45.Value
    Points.4=@n45.Value
    Points.1=@n46.Value
    Points.2=@n47.Value
    Points.3=@n48.Value
    WorldOffset=@n49.Transform
    ExecutePin=@RigVMFunction_Sequence.A
   ) {is=struct RigVMUnitNode}
   > /Script/RigVM.RigVMFunction_DebugLineNoSpace(
    B=(X=0,Y=0,Z=0)
    Color=(R=1,G=0,B=0,A=1)
    DebugDrawSettings=(DepthPriority=SDPG_Foreground,Lifetime=-1)
    Thickness=0
    bEnabled=True
    A=@n45.Value
    WorldOffset=@n49.Transform
   ) {is=struct RigVMUnitNode}
   > &RigVMFunction_DebugLineNoSpace(
    B=(X=0,Y=0,Z=0)
    Color=(R=1,G=0,B=0,A=1)
    DebugDrawSettings=(DepthPriority=SDPG_Foreground,Lifetime=-1)
    Thickness=0
    bEnabled=True
    A=@n46.Value
    WorldOffset=@n49.Transform
   )
   > &RigVMFunction_DebugLineNoSpace(
    B=(X=0,Y=0,Z=0)
    Color=(R=1,G=0,B=0,A=1)
    DebugDrawSettings=(DepthPriority=SDPG_Foreground,Lifetime=-1)
    Thickness=0
    bEnabled=True
    A=@n47.Value
    WorldOffset=@n49.Transform
   )
   > &RigVMFunction_DebugLineNoSpace(
    B=(X=0,Y=0,Z=0)
    Color=(R=1,G=0,B=0,A=1)
    DebugDrawSettings=(DepthPriority=SDPG_Foreground,Lifetime=-1)
    Thickness=0
    bEnabled=True
    A=@n48.Value
    WorldOffset=@n49.Transform
   )
   n50 = &RigVMFunction_Sequence(ExecuteContext=@RigVMFunction_Sequence.B)
   n51 = &RigUnit_CollectionChildrenArray(Parent=(Type=Control,Name="Points"), TypeToSearch=All, bDefaultChildren=True, bIncludeParent=False, bRecursive=False)
   For_Each = RigVMDispatchNode(Array=@n51.Items, ExecuteContext=@n50.A) {template="DISPATCH_RigVMDispatch_ArrayIterator(in Array,out Element,out Index,out Count,out Ratio)"}
   n52 = RigVMVariableNode() {variable=NumPoints}
   ToIndex2d = RigVMFunctionReferenceNode(Index=@For_Each.Index, NumControls=RigVMVariableNode() {variable=NumPoints}.Value)
   Less = /Script/RigVM.RigVMFunction_MathIntLess(
    B=/Script/RigVM.RigVMFunction_MathIntSub(B=1, A=@n52.Value.X) {is=struct RigVMUnitNode, template="Subtract::Execute(in A,in B,out Result)"}.Result
    A=@ToIndex2d.Index2d.X
   ) {is=struct RigVMUnitNode, template="Less::Execute(in A,in B,out Result)"}
   RigVMFunction_ControlFlowBranch = &RigVMFunction_ControlFlowBranch(ExecuteContext=@For_Each.ExecuteContext, Condition=@Less.Result)
   n53 = &RigVMFunction_ControlFlowBranch(
    Condition=&RigVMFunction_MathIntLess(
     B=&RigVMFunction_MathIntSub(B=1, A=@n52.Value.Y) {template="Subtract::Execute(in A,in B,out Result)"}.Result
     A=@ToIndex2d.Index2d.Y
    ) {template="Less::Execute(in A,in B,out Result)"}.Result
    ExecuteContext=@RigVMFunction_ControlFlowBranch.Completed
   )
   n54 = &RigUnit_GetTransform(bInitial=False, Item=@For_Each.Element)
   FromIndex2d_1 = RigVMFunctionReferenceNode(
    Index2d=(X=0,Y=0,Z=0)
    Index2d.Y=&RigVMFunction_MathIntAdd(B=1, A=@ToIndex2d.Index2d.Y) {template="Add::Execute(in A,in B,out Result)"}.Result
    Index2d.X=@ToIndex2d.Index2d.X
    NumPoints=RigVMVariableNode() {variable=NumPoints}.Value
   )
   n55 = RigVMVariableNode() {variable=LatticePointColor}
   &RigVMFunction_DebugLineNoSpace(
    DebugDrawSettings=(DepthPriority=SDPG_Foreground,Lifetime=-1)
    Thickness=0
    WorldOffset=(Rotation=(X=0,Y=0,Z=0,W=1),Translation=(X=0,Y=0,Z=0),Scale3D=(X=1,Y=1,Z=1))
    bEnabled=True
    B=@n54.Transform.Translation
    A=&RigUnit_GetTransform(
     bInitial=False
     Item=RigVMDispatchNode(Array=@n51.Items, Index=@FromIndex2d_1.Index) {template="DISPATCH_RigVMDispatch_ArrayGetAtIndex(in Array,in Index,out Element)"}.Element
    ).Transform.Translation
    ExecutePin=@n53.True
    Color=@n55.Value
   )
   Add = &RigVMFunction_MathIntAdd(B=1, A=@ToIndex2d.Index2d.X) {template="Add::Execute(in A,in B,out Result)"}
   FromIndex2d = RigVMFunctionReferenceNode(Index2d=(X=0,Y=0,Z=0), Index2d.X=@Add.Result, Index2d.Y=@ToIndex2d.Index2d.Y, NumPoints=RigVMVariableNode() {variable=NumPoints}.Value)
   RigVMFunction_DebugLineNoSpace = &RigVMFunction_DebugLineNoSpace(
    DebugDrawSettings=(DepthPriority=SDPG_Foreground,Lifetime=-1)
    Thickness=0
    WorldOffset=(Rotation=(X=0,Y=0,Z=0,W=1),Translation=(X=0,Y=0,Z=0),Scale3D=(X=1,Y=1,Z=1))
    bEnabled=True
    B=@n54.Transform.Translation
    A=&RigUnit_GetTransform(
     bInitial=False
     Item=RigVMDispatchNode(Array=@n51.Items, Index=@FromIndex2d.Index) {template="DISPATCH_RigVMDispatch_ArrayGetAtIndex(in Array,in Index,out Element)"}.Element
    ).Transform.Translation
    ExecutePin=@RigVMFunction_ControlFlowBranch.True
    Color=@n55.Value
   )
   n56 = &RigVMFunction_Sequence(ExecuteContext=@n50.B)
   n57 = &RigUnit_CollectionChildrenArray(Parent=(Type=Control,Name="Points"), TypeToSearch=All, bDefaultChildren=True, bIncludeParent=False, bRecursive=False)
   Num = RigVMDispatchNode(Array=@n57.Items) {template="DISPATCH_RigVMDispatch_ArrayGetNum(in Array,out Num)"}
   Set_Num = RigVMDispatchNode(Array=(), Num=@Num.Num, ExecuteContext=@n56.A) {template="DISPATCH_RigVMDispatch_ArraySetNum(io Array,in Num)"}
   > For_Each_1 = RigVMDispatchNode(Array=@n57.Items) {template="DISPATCH_RigVMDispatch_ArrayIterator(in Array,out Element,out Index,out Count,out Ratio)"}
   > Set_At = RigVMDispatchNode(
    Index=@For_Each_1.Index
    Element=&RigVMFunction_MathTransformMakeRelative(
     Global=&RigUnit_GetTransform(bInitial=False, Item=@For_Each_1.Element).Transform
     Parent=&RigUnit_GetTransform(Item=(Type=Control,Name="Points"), bInitial=False).Transform
    ).Local.Translation
    Array=@Set_Num.Array
   ) {template="DISPATCH_RigVMDispatch_ArraySetAtIndex(io Array,in Index,in Element)"}
   n58 = &RigVMFunction_Sequence(ExecuteContext=@n56.B)
   n59 = RigVMVariableNode() {variable=ReferenceAspectRatio}
   Radians = &RigVMFunction_MathDoubleRad(Value=RigVMVariableNode() {variable=ReferenceFieldOfView}.Value) {template="Radians::Execute(in Value,out Result)"}
   RigUnit_GetFloatAnimationChannel = &RigUnit_GetFloatAnimationChannel(Channel=Weight, bInitial=False, Control=RigVMVariableNode() {variable=LatticeRootControlName}.Value) {template="GetAnimationChannel::Execute(out Value,in Control,in Channel,in bInitial)"}
   RigUnit_GetTransform = &RigUnit_GetTransform(Item=(Type=Null,Name="CameraSpace"), bInitial=False)
   /Script/OptimusCore.RigUnit_AddOptimusDeformer(
    AspectRatioYX=(Value=0,Name="AspectRatioYX")
    DeformerGraphAsset=(DeformerGraph="&/DG_CameraSpaceLattice",Name="DeformerGraphAsset")
    HalfHeight=(Value=0,Name="HalfHeight")
    HalfWidth=(Value=0,Name="HalfWidth")
    NumPoints=(Value=(X=0,Y=0,Z=0),Name="NumPoints")
    Points=(Value=,Name="Points")
    PointsTransform=(Value=(Rotation=(X=0,Y=0,Z=0,W=1),Translation=(X=0,Y=0,Z=0),Scale3D=(X=1,Y=1,Z=1)),Name="PointsTransform")
    Settings=(ExecutionPhase=AfterDefaultDeformer,ExecutionGroup=1,DeformChildComponents=True,ExcludeChildComponentsWithTag="None",Name="Settings")
    Transform=(Value=(Rotation=(X=0,Y=0,Z=0,W=1),Translation=(X=0,Y=0,Z=0),Scale3D=(X=1,Y=1,Z=1)),Name="Transform")
    Weight=(Value=0,Name="Weight")
    XFieldOfView=(Value=0,Name="XFieldOfView")
    ZFar=(Value=0,Name="ZFar")
    ZNear=(Value=0,Name="ZNear")
    ZFar.Value=RigVMDispatchNode(Value=999999) {template="DISPATCH_RigVMDispatch_Constant(io Value)"}.Value
    AspectRatioYX.Value=&RigVMFunction_MathDoubleDiv(B=@n59.Value.X, A=@n59.Value.Y).Result
    XFieldOfView.Value=@Radians.Result
    Weight.Value=@RigUnit_GetFloatAnimationChannel.Value
    Transform.Value=@RigUnit_GetTransform.Transform
    PointsTransform.Value=&RigUnit_GetTransform(Item=(Type=Control,Name="CameraLattice"), Space=LocalSpace, bInitial=False).Transform
    ExecutePin=@n58.A
    Points.Value=@Set_Num.Array
    ZNear.Value=RigVMVariableNode() {variable=ZNear}.Value
    NumPoints.Value=RigVMVariableNode() {variable=NumPoints}.Value
    HalfWidth.Value=RigVMVariableNode() {variable=ReferenceHalfWidth}.Value
    HalfHeight.Value=RigVMVariableNode() {variable=ReferenceHalfHeight}.Value
   ) {is=struct RigVMUnitNode}
  ]}
]
hierarchy [
 CameraHandle {controlType=EULER_TRANSFORM, type=Control, loc=<0 0 0>, rot=<0 0 0 1>}
 CameraSpace {parent=CameraHandle, type=Null, loc=<0 0 0>, rot=<0.5 0.5 0.5 0.5>}
 AspectRatioX {controlType=FLOAT, parent=CameraHandle, type=Control, loc=<16 0 0>, rot=<0 0 0 1>}
 AspectRatioY {controlType=FLOAT, parent=CameraHandle, type=Control, loc=<9 0 0>, rot=<0 0 0 1>}
 FieldOfView {controlType=FLOAT, parent=CameraHandle, type=Control, loc=<37.8493 0 0>, rot=<0 0 0 1>}
]
variables [
 InteractionCount {type=int32, default=0}
 LatticePointColor {type=FLinearColor, default=(R=0.8056,G=0.8056,B=0.8056,A=0)}
 LatticeRootControlName {type=FName, default=CameraLattice}
 NumPoints {type=FIntVector, default=(X=10,Y=10,Z=0)}
 RadiusIndicatorTimeLeft {type=double, default=0}
 ReferenceAspectRatio {type=FVector, default=(X=16,Y=9,Z=0)}
 ReferenceFieldOfView {type=double, default=37.8493}
 ReferenceHalfHeight {type=double, default=0}
 ReferenceHalfWidth {type=double, default=0}
 ZNear {type=double, default=50}
]
